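# Physics asset for the PatronPoppet female medium character's face accessory part, set up around a left-side dynamic hair chain.
# Capsule bodies run from dyn_Hair_l_base through four hair bones, linked in sequence and anchored to spine_05 with angular limits, tighter at the base in swing and along the strand in twist.
# Larger capsules on spine_04, spine_05 and upperarm_l, joined by free angular constraints, provide torso and left-arm shapes around the hair.
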PhysicsAsset /BRCosmetics/Characters/Player/Female/Medium/Bodies/F_MED_PatronPoppet/Meshes/Parts/F_MED_PatronPoppet_FaceAcc_Physics
bodies {
 dyn_Hair_l_01 [{shapeType=Capsule, length=2.2607, radius=3.005, shapeName=dyn_Hair_l_01_capsule, center=<1.7967 0 0.6751>, rotation=<rot 90 0 0>}]
 dyn_Hair_l_02 [{shapeType=Capsule, length=3.3826, radius=2.505, shapeName=dyn_Hair_l_02_capsule, center=<2.163 0 0.1695>, rotation=<rot 83.9835 0 -0.0001>}]
 dyn_Hair_l_03 [{
   shapeType=Capsule
   length=3.1347
   radius=2.2805
   shapeName=dyn_Hair_l_03_capsule
   center=<2.1422 0 0.6064>
   rotation=<rot 83.6116 179.9998 179.9998>
  }]
 dyn_Hair_l_04 [{
   shapeType=Capsule
   length=3.2233
   radius=1.7003
   shapeName=dyn_Hair_l_04_capsule
   center=<1.8074 0.7691 0.1064>
   rotation=<rot 81.8054 23.0514 0.0002>
  }]
 dyn_Hair_l_base [{shapeType=Capsule, length=0.26, radius=2.25, shapeName=dyn_Hair_l_base_capsule, rotation=<rot 90 0 0>}]
 spine_04 [{shapeType=Capsule, length=19.5, radius=10.5, shapeName=spine_04_capsule, center=<-0.5867 0.561 0>}]
 spine_05 [{shapeType=Capsule, length=21, radius=9, shapeName=spine_05_capsule, center=<-0.3981 -2.0489 0>}]
 upperarm_l [
  {shapeType=Capsule, length=25.75, radius=4.755, shapeName=upperarm_l_capsule, center=<10.5007 0 0>, rotation=<rot -0.0001 -89.9999 90>}
  {
   shapeType=Capsule
   length=8.6984
   radius=4.755
   shapeName=upperarm_l_capsule
   center=<17.3681 2.6063 -2.5913>
   rotation=<rot -0.0001 -89.9999 90>
  }
 ]
}
constraints [
 {
  bone1Name=spine_05
  bone2Name=spine_04
  swing1LimitDegrees=45
  swing1Motion=Free
  swing2LimitDegrees=45
  swing2Motion=Free
  twistLimitDegrees=45
  twistMotion=Free
 }
 {bone1Name=dyn_Hair_l_04, bone2Name=dyn_Hair_l_03, swing1LimitDegrees=25, swing2LimitDegrees=25, twistLimitDegrees=10}
 {bone1Name=dyn_Hair_l_03, bone2Name=dyn_Hair_l_02, swing1LimitDegrees=25, swing2LimitDegrees=25, twistLimitDegrees=10}
 {bone1Name=dyn_Hair_l_02, bone2Name=dyn_Hair_l_01, swing1LimitDegrees=25, swing2LimitDegrees=25, twistLimitDegrees=10}
 {bone1Name=dyn_Hair_l_01, bone2Name=dyn_Hair_l_base, swing1LimitDegrees=25, swing2LimitDegrees=25, twistLimitDegrees=10}
 {bone1Name=dyn_Hair_l_base, bone2Name=spine_05, swing1LimitDegrees=15, swing2LimitDegrees=15, twistLimitDegrees=15}
 {
  bone1Name=upperarm_l
  bone2Name=spine_05
  swing1LimitDegrees=45
  swing1Motion=Free
  swing2LimitDegrees=45
  swing2Motion=Free
  twistLimitDegrees=45
  twistMotion=Free
 }
]
