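# Post-process style AnimBlueprint for the PatronPoppet female cosmetic's face accessory mesh, adding secondary motion to hair and hat.
# Copied parent and body poses blend per bone, then feed a Control Rig driven by movement, weapon, cape and front-end states.
# Chained Anim Dynamics nodes with springs and arm sphere limits, plus a RigidBody node, simulate hair, then hat.
# Bool blends select cached poses before additive bone tweaks and a constraint finish the output.
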AnimBlueprint /BRCosmetics/Characters/Player/Female/Medium/Bodies/F_MED_PatronPoppet/Meshes/Parts/F_MED_PatronPoppet_FaceAcc_AnimBP
target_skeleton=&/F_MED_PatronPoppet_FaceAcc_Skeleton
graphs [
 AnimGraph {nodes [
   n1 = Blend Poses by bool(BlendPose_0=Use cached pose 'CRCache_Cape'(), BlendPose_1=Use cached pose 'CRCache_HatDyn'()) {blendType=Linear}
   n2 = Use cached pose 'CRCache_BP'()
   n3 = "Transform (Modify) Bone - Bone: dyn_Hair_BackL_02"!(
    ComponentPose=Local To Component(LocalPose=Blend Poses by bool(BlendPose_0=@n2, BlendPose_1=@n1) {blendType=Linear})
    bAlphaBoolEnabled:bool=True
    Rotation="5,0,0"
   ) -> (Pose) {
    alphaInputType=Bool
    rotationMode=BMM_Additive
    rotationSpace=BCS_BoneSpace
    scaleSpace=BCS_BoneSpace
    translationMode=BMM_Additive
    translationSpace=BCS_BoneSpace
    alphaBoolBlend {blendInTime=0.1, blendOption=Linear, blendOutTime=0.1}
    boneToModify {boneName=dyn_Hair_BackL_02}
    rotation {pitch=5}
   }
   n4 = "Transform (Modify) Bone - Bone: dyn_Hair_BackR_02"!(ComponentPose=@n3, bAlphaBoolEnabled:bool=True, Rotation="5,0,0") -> (Pose) {
    alphaInputType=Bool
    rotationMode=BMM_Additive
    rotationSpace=BCS_BoneSpace
    scaleSpace=BCS_BoneSpace
    translationMode=BMM_Additive
    translationSpace=BCS_BoneSpace
    alphaBoolBlend {blendInTime=0.1, blendOption=Linear, blendOutTime=0.1}
    boneToModify {boneName=dyn_Hair_BackR_02}
    rotation {pitch=5}
   }
   n5 = "Transform (Modify) Bone - Bone: dyn_hat_bk01"!(ComponentPose=@n4, bAlphaBoolEnabled:bool=True, Rotation="15,0,0") -> (Pose) {
    alphaInputType=Bool
    rotationMode=BMM_Additive
    rotationSpace=BCS_BoneSpace
    scaleSpace=BCS_BoneSpace
    translationMode=BMM_Additive
    translationSpace=BCS_BoneSpace
    alphaBoolBlend {blendInTime=0.1, blendOption=Linear, blendOutTime=0.1}
    boneToModify {boneName=dyn_hat_bk01}
    rotation {pitch=15}
   }
   n6 = "Transform (Modify) Bone - Bone: dyn_Hair_Back_02"!(ComponentPose=@n5, bAlphaBoolEnabled:bool=True, Rotation="5,0,0") -> (Pose) {
    alphaInputType=Bool
    rotationMode=BMM_Additive
    rotationSpace=BCS_BoneSpace
    scaleSpace=BCS_BoneSpace
    translationMode=BMM_Additive
    translationSpace=BCS_BoneSpace
    alphaBoolBlend {blendInTime=0.1, blendOption=Linear, blendOutTime=0.1}
    boneToModify {boneName=dyn_Hair_Back_02}
    rotation {pitch=5}
   }
   n7 = "Transform (Modify) Bone - Bone: dyn_Hair_l_base"!(ComponentPose=@n6, bAlphaBoolEnabled:bool=True, Rotation="0,8,0") -> (Pose) {
    alphaInputType=Bool
    rotationMode=BMM_Additive
    rotationSpace=BCS_BoneSpace
    scaleMode=BMM_Additive
    scaleSpace=BCS_BoneSpace
    translationMode=BMM_Additive
    translationSpace=BCS_BoneSpace
    alphaBoolBlend {blendOption=Linear}
    boneToModify {boneName=dyn_Hair_l_base}
    rotation {yaw=8}
   }
   n8 = "Transform (Modify) Bone - Bone: dyn_Hair_l_02"!(ComponentPose=@n7, bAlphaBoolEnabled:bool=True, Rotation="-3,0,0") -> (Pose) {
    alphaInputType=Bool
    rotationMode=BMM_Additive
    rotationSpace=BCS_BoneSpace
    scaleMode=BMM_Additive
    scaleSpace=BCS_BoneSpace
    translationMode=BMM_Additive
    translationSpace=BCS_BoneSpace
    alphaBoolBlend {blendOption=Linear}
    boneToModify {boneName=dyn_Hair_l_02}
    rotation {pitch=-3}
    translation {x=0}
   }
   n9 = "Transform (Modify) Bone - Bone: dyn_Hair_l_04"!(ComponentPose=@n8, bAlphaBoolEnabled:bool=True, Rotation="25,-15,0") -> (Pose) {
    alphaInputType=Bool
    rotationMode=BMM_Additive
    rotationSpace=BCS_BoneSpace
    scaleMode=BMM_Additive
    scaleSpace=BCS_BoneSpace
    translationMode=BMM_Additive
    translationSpace=BCS_BoneSpace
    alphaBoolBlend {blendOption=Linear}
    boneToModify {boneName=dyn_Hair_l_04}
    rotation {pitch=25, yaw=-15}
   }
   n10 = "Constraint - dyn_Hair_BackR_03a (PP)"!(ComponentPose=@n9, bAlphaBoolEnabled:bool=True, ConstraintWeights_0=0.2, ConstraintWeights_1:float=0.8) -> (Pose) {
    alphaInputType=Bool
    alphaBoolBlend {blendInTime=0.1, blendOption=Linear, blendOutTime=0.1}
    boneToModify {boneName=dyn_Hair_BackR_03a}
    constraintSetup [{targetBone {boneName=dyn_Hair_BackR_03a}}, {targetBone {boneName=clavicle_r}}]
    constraintWeights [1,1]
   }
   n11 = "Transform (Modify) Bone - Bone: dyn_Hair_BackR_03a"!(ComponentPose=@n10, bAlphaBoolEnabled:bool=True, Rotation="-5,0,20") -> (Pose) {
    alphaInputType=Bool
    rotationMode=BMM_Additive
    rotationSpace=BCS_BoneSpace
    scaleMode=BMM_Additive
    scaleSpace=BCS_BoneSpace
    translationMode=BMM_Additive
    translationSpace=BCS_BoneSpace
    alphaBoolBlend {blendInTime=0.1, blendOption=Linear, blendOutTime=0.1}
    boneToModify {boneName=dyn_Hair_BackR_03a}
    rotation {pitch=-5, roll=20}
   }
   Output Pose(Result=Component To Local(ComponentPose=@n11)) {layerGroup=None}
   n12 = Layered blend per bone(
    BasePose=Copy Pose From Mesh(SourceMeshComponent=Get Skel_Mesh()) {bUseAttachedParent=true}
    BlendPoses_0=Copy Pose From Mesh(SourceMeshComponent=Get Body_Mesh())
   ) {layerSetup [{branchFilters [{boneName=clavicle_l}, {boneName=clavicle_r}, {boneName=calf_l}, {boneName=calf_r}]}]}
   n13 = Control Rig!(
    Source=@n12
    Alpha
    isSkydiveFloating:double=Get isSkydiveFloating() -> (isSkydiveFloating:double)
    isHangGliding:double=Get isHangGliding() -> (isHangGliding:double)
    isCrouching:double=Get isCrouching() -> (isCrouching:double)
    isGunEquipped:double=Get isGunOut() -> (isGunOut:double)
    isTargeting:double=Get isTargeting() -> (isTargeting:double)
    isFrontEnd:bool=Get bIsInFrontEnd()
    isCapeEquipped:double=Get is_f_WearingCape() -> (is_f_WearingCape:double)
   ) -> (Pose) {
    alphaBoolBlend {blendOption=Linear}
    controlRigAssetReference {blueprintRigClass=&/F_MED_PatronPoppet_FaceAcc_CtrlRig.F_MED_PatronPoppet_FaceAcc_CtrlRig_C}
   }
   Save cached pose 'CRCache_HairPose'(Pose=@n13) {cacheName=CRCache_HairPose}
   n14 = Anim Dynamics - Bone: dyn_Hair_Back_02!(ComponentPose=Local To Component(LocalPose=Use cached pose 'CRCache_HairPose'()), Alpha=0.2, GravityOverride:Vector="0,0,-980") -> (Pose) {
    angularDampingOverride=0.99
    angularSpringConstant=75
    bAngularSpring=true
    bGravityOverrideInSimSpace=true
    bOverrideAngularDamping=true
    bUseGravityOverride=true
    simulationSpace=RootRelative
    alphaBoolBlend {blendOption=Linear}
    boundBone {boneName=dyn_Hair_Back_02}
    componentAppliedLinearAccClamp=<3500 3500 3000>
    componentLinearAccScale=<0.5 0.5 0.5>
    componentLinearVelScale=<0.65 0.65 0.65>
    gravityOverride {z=-980}
    physicsBodyDefinitions [{
      collisionType=CustomSphere
      sphereCollisionRadius=2
      boundBone {boneName=dyn_Hair_Back_02}
      boxExtents=<40 40 40>
      constraintSetup {angularLimitsMax=<0 0 2.5>, angularLimitsMin=<0 -5 -2.5>, angularTarget=<0 0 0>}
      localJointOffset=<5 0 0>
     }]
    relativeSpaceBone {boneName=root}
   }
   n15 = Anim Dynamics - Bone: dyn_Hair_Back_03!(SimulationSpace:AnimPhysSimSpaceType Enum, ComponentPose=@n14, Alpha=1) -> (Pose) {
    angularDampingOverride=0.99
    angularSpringConstant=75
    bAngularSpring=true
    bGravityOverrideInSimSpace=true
    bOverrideAngularDamping=true
    bUseGravityOverride=true
    simulationSpace=RootRelative
    alphaBoolBlend {blendOption=Linear}
    alphaScaleBiasClamp {bInterpResult=true}
    boundBone {boneName=dyn_Hair_Back_03}
    componentAppliedLinearAccClamp=<3500 3500 3000>
    componentLinearAccScale=<0.5 0.5 0.5>
    componentLinearVelScale=<0.65 0.65 0.65>
    gravityOverride {z=-980}
    physicsBodyDefinitions [{
      collisionType=CustomSphere
      sphereCollisionRadius=2
      boundBone {boneName=dyn_Hair_Back_03}
      boxExtents=<25 25 25>
      constraintSetup {angularLimitsMax=<0 2 5>, angularLimitsMin=<0 -20 -5>, angularTarget=<0 0 0>}
      localJointOffset=<5 0 0>
     }]
    relativeSpaceBone {boneName=root}
   }
   n16 = Anim Dynamics - Bone: dyn_Hair_Back_04!(SimulationSpace:AnimPhysSimSpaceType Enum, ComponentPose=@n15, Alpha=1) -> (Pose) {
    angularDampingOverride=0.99
    angularSpringConstant=75
    bAngularSpring=true
    bGravityOverrideInSimSpace=true
    bOverrideAngularDamping=true
    bUseGravityOverride=true
    simulationSpace=RootRelative
    alphaBoolBlend {blendOption=Linear}
    alphaScaleBiasClamp {bInterpResult=true}
    boundBone {boneName=dyn_Hair_Back_04}
    componentAppliedLinearAccClamp=<3500 3500 3000>
    componentLinearAccScale=<0.5 0.5 0.5>
    componentLinearVelScale=<0.65 0.65 0.65>
    gravityOverride {z=-980}
    physicsBodyDefinitions [{
      collisionType=CustomSphere
      sphereCollisionRadius=2
      boundBone {boneName=dyn_Hair_Back_04}
      boxExtents=<23 23 23>
      constraintSetup {angularLimitsMax=<2 0 7.5>, angularLimitsMin=<-2 -25 -7.5>, angularTarget=<0 0 0>}
      localJointOffset=<5 0 0>
     }]
    relativeSpaceBone {boneName=root}
   }
   n17 = Anim Dynamics - Bone: dyn_Hair_Back_05!(SimulationSpace:AnimPhysSimSpaceType Enum, ComponentPose=@n16, Alpha=1) -> (Pose) {
    angularDampingOverride=0.99
    angularSpringConstant=75
    bAngularSpring=true
    bGravityOverrideInSimSpace=true
    bOverrideAngularDamping=true
    bUseGravityOverride=true
    bUseSphericalLimits=true
    simulationSpace=RootRelative
    alphaBoolBlend {blendOption=Linear}
    alphaScaleBiasClamp {bInterpResult=true}
    boundBone {boneName=dyn_Hair_Back_05}
    componentAppliedLinearAccClamp=<3500 3500 3000>
    componentLinearAccScale=<0.5 0.5 0.5>
    componentLinearVelScale=<0.65 0.65 0.65>
    gravityOverride {z=-980}
    physicsBodyDefinitions [{
      collisionType=CustomSphere
      sphereCollisionRadius=2
      boundBone {boneName=dyn_Hair_Back_05}
      boxExtents=<23 23 23>
      constraintSetup {angularLimitsMax=<2 0 5>, angularLimitsMin=<-2 -25 -5>, angularTarget=<0 0 0>}
      localJointOffset=<5 0 0>
     }]
    relativeSpaceBone {boneName=root}
    sphericalLimits [
     {limitRadius=0, drivingBone {boneName=None}, sphereLocalOffset=<0 0 0>}
     {limitRadius=0, drivingBone {boneName=None}, sphereLocalOffset=<0 0 0>}
    ]
   }
   n18 = Anim Dynamics - Bone: dyn_Hair_Back_06!(SimulationSpace:AnimPhysSimSpaceType Enum, ComponentPose=@n17, Alpha=1) -> (Pose) {
    angularDampingOverride=0.99
    angularSpringConstant=75
    bAngularSpring=true
    bGravityOverrideInSimSpace=true
    bOverrideAngularDamping=true
    bUseGravityOverride=true
    simulationSpace=RootRelative
    alphaBoolBlend {blendOption=Linear}
    alphaScaleBiasClamp {bInterpResult=true}
    boundBone {boneName=dyn_Hair_Back_06}
    componentAppliedLinearAccClamp=<3500 3500 3000>
    componentLinearAccScale=<0.5 0.5 0.5>
    componentLinearVelScale=<0.65 0.65 0.65>
    gravityOverride {z=-980}
    physicsBodyDefinitions [{
      collisionType=CustomSphere
      sphereCollisionRadius=2
      boundBone {boneName=dyn_Hair_Back_06}
      boxExtents=<23 23 23>
      constraintSetup {angularLimitsMax=<2 0 2.5>, angularLimitsMin=<-2 -25 -2.5>, angularTarget=<0 0 0>}
      localJointOffset=<5 0 0>
     }]
    relativeSpaceBone {boneName=root}
    sphericalLimits [
     {limitRadius=18, drivingBone {boneName=upperarm_r}, sphereLocalOffset=<5 -14 0>}
     {limitRadius=18, drivingBone {boneName=upperarm_l}, sphereLocalOffset=<5 14 0>}
    ]
   }
   n19 = Anim Dynamics - Bone: dyn_Hair_Back_07!(SimulationSpace:AnimPhysSimSpaceType Enum, ComponentPose=@n18, Alpha=1) -> (Pose) {
    angularDampingOverride=0.99
    angularSpringConstant=75
    bAngularSpring=true
    bGravityOverrideInSimSpace=true
    bOverrideAngularDamping=true
    bUseGravityOverride=true
    simulationSpace=RootRelative
    alphaBoolBlend {blendOption=Linear}
    alphaScaleBiasClamp {bInterpResult=true}
    boundBone {boneName=dyn_Hair_Back_07}
    componentAppliedLinearAccClamp=<3500 3500 3000>
    componentLinearAccScale=<0.5 0.5 0.5>
    componentLinearVelScale=<0.65 0.65 0.65>
    gravityOverride {z=-980}
    physicsBodyDefinitions [{
      collisionType=CustomSphere
      sphereCollisionRadius=2
      boundBone {boneName=dyn_Hair_Back_07}
      boxExtents=<23 23 23>
      constraintSetup {angularLimitsMax=<1 0 1.5>, angularLimitsMin=<-1 0 -1.5>, angularTarget=<0 0 0>}
      localJointOffset=<5 0 0>
     }]
    relativeSpaceBone {boneName=root}
    sphericalLimits [
     {limitRadius=18, drivingBone {boneName=upperarm_r}, sphereLocalOffset=<5 -14 0>}
     {limitRadius=18, drivingBone {boneName=upperarm_l}, sphereLocalOffset=<5 14 0>}
    ]
   }
   n20 = Anim Dynamics - Bone: dyn_Hair_BackR_02!(ComponentPose=@n19, Alpha=0.2, GravityOverride:Vector="0,0,-980") -> (Pose) {
    angularDampingOverride=0.99
    angularSpringConstant=75
    bAngularSpring=true
    bGravityOverrideInSimSpace=true
    bOverrideAngularDamping=true
    bUseGravityOverride=true
    simulationSpace=RootRelative
    alphaBoolBlend {blendOption=Linear}
    boundBone {boneName=dyn_Hair_BackR_02}
    componentAppliedLinearAccClamp=<3500 3500 3000>
    componentLinearAccScale=<0.5 0.5 0.5>
    componentLinearVelScale=<0.65 0.65 0.65>
    gravityOverride {z=-980}
    physicsBodyDefinitions [{
      collisionType=CustomSphere
      sphereCollisionRadius=2
      boundBone {boneName=dyn_Hair_BackR_02}
      boxExtents=<40 40 40>
      constraintSetup {angularLimitsMax=<0 0 2.5>, angularLimitsMin=<0 -5 -2.5>, angularTarget=<0 0 0>}
      localJointOffset=<5 0 0>
     }]
    relativeSpaceBone {boneName=root}
   }
   n21 = Anim Dynamics - Bone: dyn_Hair_BackR_03!(SimulationSpace:AnimPhysSimSpaceType Enum, ComponentPose=@n20, Alpha=1) -> (Pose) {
    angularDampingOverride=0.99
    angularSpringConstant=75
    bAngularSpring=true
    bGravityOverrideInSimSpace=true
    bOverrideAngularDamping=true
    bUseGravityOverride=true
    bUseSphericalLimits=true
    simulationSpace=RootRelative
    alphaBoolBlend {blendOption=Linear}
    alphaScaleBiasClamp {bInterpResult=true}
    boundBone {boneName=dyn_Hair_BackR_03}
    componentAppliedLinearAccClamp=<3500 3500 3000>
    componentLinearAccScale=<0.5 0.5 0.5>
    componentLinearVelScale=<0.65 0.65 0.65>
    gravityOverride {z=-980}
    physicsBodyDefinitions [{
      collisionType=CustomSphere
      sphereCollisionRadius=2
      boundBone {boneName=dyn_Hair_BackR_03}
      boxExtents=<25 25 25>
      constraintSetup {angularLimitsMax=<0 2 5>, angularLimitsMin=<0 -20 -5>, angularTarget=<0 0 0>}
      localJointOffset=<5 0 0>
     }]
    relativeSpaceBone {boneName=root}
    sphericalLimits [{limitRadius=18, drivingBone {boneName=None}, sphereLocalOffset=<-5 -13 0>}]
   }
   n22 = Anim Dynamics - Bone: dyn_Hair_BackR_04!(SimulationSpace:AnimPhysSimSpaceType Enum, ComponentPose=@n21, Alpha=1) -> (Pose) {
    angularDampingOverride=0.99
    angularSpringConstant=75
    bAngularSpring=true
    bGravityOverrideInSimSpace=true
    bOverrideAngularDamping=true
    bUseGravityOverride=true
    simulationSpace=RootRelative
    alphaBoolBlend {blendOption=Linear}
    alphaScaleBiasClamp {bInterpResult=true}
    boundBone {boneName=dyn_Hair_BackR_04}
    componentAppliedLinearAccClamp=<3500 3500 3000>
    componentLinearAccScale=<0.5 0.5 0.5>
    componentLinearVelScale=<0.65 0.65 0.65>
    gravityOverride {z=-980}
    physicsBodyDefinitions [{
      collisionType=CustomSphere
      sphereCollisionRadius=2
      boundBone {boneName=dyn_Hair_BackR_04}
      boxExtents=<23 23 23>
      constraintSetup {angularLimitsMax=<2 0 7.5>, angularLimitsMin=<-2 -25 -7.5>, angularTarget=<0 0 0>}
      localJointOffset=<5 0 0>
     }]
    relativeSpaceBone {boneName=root}
   }
   n23 = Anim Dynamics - Bone: dyn_Hair_BackR_05!(SimulationSpace:AnimPhysSimSpaceType Enum, ComponentPose=@n22, Alpha=1) -> (Pose) {
    angularDampingOverride=0.99
    angularSpringConstant=75
    bAngularSpring=true
    bGravityOverrideInSimSpace=true
    bOverrideAngularDamping=true
    bUseGravityOverride=true
    bUseSphericalLimits=true
    simulationSpace=RootRelative
    alphaBoolBlend {blendOption=Linear}
    alphaScaleBiasClamp {bInterpResult=true}
    boundBone {boneName=dyn_Hair_BackR_05}
    componentAppliedLinearAccClamp=<3500 3500 3000>
    componentLinearAccScale=<0.5 0.5 0.5>
    componentLinearVelScale=<0.65 0.65 0.65>
    gravityOverride {z=-980}
    physicsBodyDefinitions [{
      collisionType=CustomSphere
      sphereCollisionRadius=2
      boundBone {boneName=dyn_Hair_BackR_05}
      boxExtents=<23 23 23>
      constraintSetup {angularLimitsMax=<2 0 5>, angularLimitsMin=<-2 -25 -5>, angularTarget=<0 0 0>}
      localJointOffset=<5 0 0>
     }]
    relativeSpaceBone {boneName=root}
    sphericalLimits [
     {limitRadius=18, drivingBone {boneName=upperarm_r}, sphereLocalOffset=<-10 -12 0>}
     {limitRadius=0, drivingBone {boneName=None}, sphereLocalOffset=<0 0 0>}
    ]
   }
   n24 = Anim Dynamics - Bone: dyn_Hair_BackR_06!(SimulationSpace:AnimPhysSimSpaceType Enum, ComponentPose=@n23, Alpha=1) -> (Pose) {
    angularDampingOverride=0.99
    angularSpringConstant=75
    bAngularSpring=true
    bGravityOverrideInSimSpace=true
    bOverrideAngularDamping=true
    bUseGravityOverride=true
    bUseSphericalLimits=true
    simulationSpace=RootRelative
    alphaBoolBlend {blendOption=Linear}
    alphaScaleBiasClamp {bInterpResult=true}
    boundBone {boneName=dyn_Hair_BackR_06}
    componentAppliedLinearAccClamp=<3500 3500 3000>
    componentLinearAccScale=<0.5 0.5 0.5>
    componentLinearVelScale=<0.65 0.65 0.65>
    gravityOverride {z=-980}
    physicsBodyDefinitions [{
      collisionType=CustomSphere
      sphereCollisionRadius=2
      boundBone {boneName=dyn_Hair_BackR_06}
      boxExtents=<23 23 23>
      constraintSetup {angularLimitsMax=<2 0 2.5>, angularLimitsMin=<-2 -25 -2.5>, angularTarget=<0 0 0>}
      localJointOffset=<5 0 0>
     }]
    relativeSpaceBone {boneName=root}
    sphericalLimits [
     {limitRadius=18, drivingBone {boneName=upperarm_r}, sphereLocalOffset=<-10 -12 0>}
     {limitRadius=0, drivingBone {boneName=None}, sphereLocalOffset=<0 0 0>}
    ]
   }
   n25 = Anim Dynamics - Bone: dyn_Hair_BackR_03a!(SimulationSpace:AnimPhysSimSpaceType Enum, ComponentPose=@n24, Alpha=1) -> (Pose) {
    angularDampingOverride=0.99
    angularSpringConstant=75
    bAngularSpring=true
    bGravityOverrideInSimSpace=true
    bOverrideAngularDamping=true
    bUseGravityOverride=true
    bUseSphericalLimits=true
    simulationSpace=RootRelative
    alphaBoolBlend {blendOption=Linear}
    alphaScaleBiasClamp {bInterpResult=true}
    boundBone {boneName=dyn_Hair_BackR_03a}
    componentAppliedLinearAccClamp=<3500 3500 3000>
    componentLinearAccScale=<0.5 0.5 0.5>
    componentLinearVelScale=<0.65 0.65 0.65>
    gravityOverride {z=-980}
    physicsBodyDefinitions [{
      collisionType=CustomSphere
      sphereCollisionRadius=2
      boundBone {boneName=dyn_Hair_BackR_03a}
      boxExtents=<23 23 23>
      constraintSetup {angularLimitsMax=<1 2 2>, angularLimitsMin=<-1 -20 -2>, angularTarget=<0 0 0>}
      localJointOffset=<2 0 0>
     }]
    relativeSpaceBone {boneName=root}
    sphericalLimits [
     {limitRadius=5, drivingBone {boneName=clavicle_r}, sphereLocalOffset=<-5 -3 0>}
     {limitRadius=0, drivingBone {boneName=None}, sphereLocalOffset=<-2 0 -3>}
    ]
   }
   n26 = Anim Dynamics - Bone: dyn_Hair_BackL_02!(ComponentPose=@n25, Alpha=0.2, GravityOverride:Vector="0,0,-980") -> (Pose) {
    angularDampingOverride=0.99
    angularSpringConstant=75
    bAngularSpring=true
    bGravityOverrideInSimSpace=true
    bOverrideAngularDamping=true
    bUseGravityOverride=true
    simulationSpace=RootRelative
    alphaBoolBlend {blendOption=Linear}
    boundBone {boneName=dyn_Hair_BackL_02}
    componentAppliedLinearAccClamp=<3500 3500 3000>
    componentLinearAccScale=<0.5 0.5 0.5>
    componentLinearVelScale=<0.65 0.65 0.65>
    gravityOverride {z=-980}
    physicsBodyDefinitions [{
      collisionType=CustomSphere
      sphereCollisionRadius=2
      boundBone {boneName=dyn_Hair_BackL_02}
      boxExtents=<40 40 40>
      constraintSetup {angularLimitsMax=<0 0 2.5>, angularLimitsMin=<0 -5 -2.5>, angularTarget=<0 0 0>}
      localJointOffset=<5 0 0>
     }]
    relativeSpaceBone {boneName=root}
   }
   n27 = Anim Dynamics - Bone: dyn_Hair_BackL_03!(SimulationSpace:AnimPhysSimSpaceType Enum, ComponentPose=@n26, Alpha=1) -> (Pose) {
    angularDampingOverride=0.99
    angularSpringConstant=75
    bAngularSpring=true
    bGravityOverrideInSimSpace=true
    bOverrideAngularDamping=true
    bUseGravityOverride=true
    simulationSpace=RootRelative
    alphaBoolBlend {blendOption=Linear}
    alphaScaleBiasClamp {bInterpResult=true}
    boundBone {boneName=dyn_Hair_BackL_03}
    componentAppliedLinearAccClamp=<3500 3500 3000>
    componentLinearAccScale=<0.5 0.5 0.5>
    componentLinearVelScale=<0.65 0.65 0.65>
    gravityOverride {z=-980}
    physicsBodyDefinitions [{
      collisionType=CustomSphere
      sphereCollisionRadius=2
      boundBone {boneName=dyn_Hair_BackL_03}
      boxExtents=<25 25 25>
      constraintSetup {angularLimitsMax=<0 2 5>, angularLimitsMin=<0 -20 -5>, angularTarget=<0 0 0>}
      localJointOffset=<5 0 0>
     }]
    relativeSpaceBone {boneName=root}
   }
   n28 = Anim Dynamics - Bone: dyn_Hair_BackL_04!(SimulationSpace:AnimPhysSimSpaceType Enum, ComponentPose=@n27, Alpha=1) -> (Pose) {
    angularDampingOverride=0.99
    angularSpringConstant=75
    bAngularSpring=true
    bGravityOverrideInSimSpace=true
    bOverrideAngularDamping=true
    bUseGravityOverride=true
    bUseSphericalLimits=true
    simulationSpace=RootRelative
    alphaBoolBlend {blendOption=Linear}
    alphaScaleBiasClamp {bInterpResult=true}
    boundBone {boneName=dyn_Hair_BackL_04}
    componentAppliedLinearAccClamp=<3500 3500 3000>
    componentLinearAccScale=<0.5 0.5 0.5>
    componentLinearVelScale=<0.65 0.65 0.65>
    gravityOverride {z=-980}
    physicsBodyDefinitions [{
      collisionType=CustomSphere
      sphereCollisionRadius=2
      boundBone {boneName=dyn_Hair_BackL_04}
      boxExtents=<23 23 23>
      constraintSetup {angularLimitsMax=<2 0 7.5>, angularLimitsMin=<-2 -25 -7.5>, angularTarget=<0 0 0>}
      localJointOffset=<5 0 0>
     }]
    relativeSpaceBone {boneName=root}
    sphericalLimits [
     {limitRadius=0, drivingBone {boneName=None}, sphereLocalOffset=<0 0 0>}
     {limitRadius=0, drivingBone {boneName=None}, sphereLocalOffset=<0 0 0>}
    ]
   }
   n29 = Anim Dynamics - Bone: dyn_Hair_BackL_05!(SimulationSpace:AnimPhysSimSpaceType Enum, ComponentPose=@n28, Alpha=1) -> (Pose) {
    angularDampingOverride=0.99
    angularSpringConstant=75
    bAngularSpring=true
    bGravityOverrideInSimSpace=true
    bOverrideAngularDamping=true
    bUseGravityOverride=true
    bUseSphericalLimits=true
    simulationSpace=RootRelative
    alphaBoolBlend {blendOption=Linear}
    alphaScaleBiasClamp {bInterpResult=true}
    boundBone {boneName=dyn_Hair_BackL_05}
    componentAppliedLinearAccClamp=<3500 3500 3000>
    componentLinearAccScale=<0.5 0.5 0.5>
    componentLinearVelScale=<0.65 0.65 0.65>
    gravityOverride {z=-980}
    physicsBodyDefinitions [{
      collisionType=CustomSphere
      sphereCollisionRadius=2
      boundBone {boneName=dyn_Hair_BackL_05}
      boxExtents=<23 23 23>
      constraintSetup {angularLimitsMax=<2 0 5>, angularLimitsMin=<-2 -25 -5>, angularTarget=<0 0 0>}
      localJointOffset=<5 0 0>
     }]
    relativeSpaceBone {boneName=root}
    sphericalLimits [
     {limitRadius=18, drivingBone {boneName=upperarm_l}, sphereLocalOffset=<5 14 0>}
     {limitRadius=0, drivingBone {boneName=None}, sphereLocalOffset=<0 0 0>}
    ]
   }
   n30 = Anim Dynamics - Bone: dyn_Hair_BackL_06!(SimulationSpace:AnimPhysSimSpaceType Enum, ComponentPose=@n29, Alpha=1) -> (Pose) {
    angularDampingOverride=0.99
    angularSpringConstant=75
    bAngularSpring=true
    bGravityOverrideInSimSpace=true
    bOverrideAngularDamping=true
    bUseGravityOverride=true
    bUseSphericalLimits=true
    simulationSpace=RootRelative
    alphaBoolBlend {blendOption=Linear}
    alphaScaleBiasClamp {bInterpResult=true}
    boundBone {boneName=dyn_Hair_BackL_06}
    componentAppliedLinearAccClamp=<3500 3500 3000>
    componentLinearAccScale=<0.5 0.5 0.5>
    componentLinearVelScale=<0.65 0.65 0.65>
    gravityOverride {z=-980}
    physicsBodyDefinitions [{
      collisionType=CustomSphere
      sphereCollisionRadius=2
      boundBone {boneName=dyn_Hair_BackL_06}
      boxExtents=<23 23 23>
      constraintSetup {angularLimitsMax=<2 0 2.5>, angularLimitsMin=<-2 -25 -2.5>, angularTarget=<0 0 0>}
      localJointOffset=<5 0 0>
     }]
    relativeSpaceBone {boneName=root}
    sphericalLimits [
     {limitRadius=18, drivingBone {boneName=upperarm_l}, sphereLocalOffset=<5 14 0>}
     {limitRadius=0, drivingBone {boneName=None}, sphereLocalOffset=<0 0 0>}
    ]
   }
   Save cached pose 'CRCache_HairDynPose'(Pose=Component To Local(ComponentPose=@n30)) {cacheName=CRCache_HairDynPose}
   n31 = Two Way Blend(A=Use cached pose 'CRCache_HairDynPose'(), B=Use cached pose 'CRCache_HairPose'(), Alpha=0.2) {blendNode {alphaBoolBlend {blendOption=Linear}}}
   n32 = Use cached pose 'CRCache_HairDynPose'()
   n33 = Anim Dynamics - Bone: dyn_Hair_a!(
    SimulationSpace:AnimPhysSimSpaceType Enum
    ComponentPose=Local To Component(LocalPose=Blend Poses by bool(BlendPose_0=@n31, BlendPose_1=@n32) {blendType=Linear})
    Alpha=1
   ) -> (Pose) {
    angularDampingOverride=0.99
    angularSpringConstant=100
    bAngularSpring=true
    bGravityOverrideInSimSpace=true
    bOverrideAngularDamping=true
    bUseGravityOverride=true
    simulationSpace=RootRelative
    alphaBoolBlend {blendOption=Linear}
    boundBone {boneName=dyn_Hair_a}
    componentAppliedLinearAccClamp=<3500 3500 2500>
    componentLinearAccScale=<0.5 0.5 0.35>
    componentLinearVelScale=<0.55 0.55 0.4>
    gravityOverride {z=-980}
    physicsBodyDefinitions [{
      collisionType=CustomSphere
      sphereCollisionRadius=2
      boundBone {boneName=dyn_Hair_a}
      boxExtents=<23 23 23>
      constraintSetup {angularLimitsMax=<2 1 10>, angularLimitsMin=<-2 -5 -10>, angularTarget=<0 0 0>}
      localJointOffset=<4 0 0>
     }]
    relativeSpaceBone {boneName=root}
   }
   n34 = Anim Dynamics - Bone: dyn_Hair_b01!(SimulationSpace:AnimPhysSimSpaceType Enum, ComponentPose=@n33, Alpha=1) -> (Pose) {
    angularDampingOverride=0.99
    angularSpringConstant=100
    bAngularSpring=true
    bGravityOverrideInSimSpace=true
    bOverrideAngularDamping=true
    bUseGravityOverride=true
    simulationSpace=RootRelative
    alphaBoolBlend {blendOption=Linear}
    boundBone {boneName=dyn_Hair_b01}
    componentAppliedLinearAccClamp=<3500 3500 2500>
    componentLinearAccScale=<0.5 0.5 0.35>
    componentLinearVelScale=<0.55 0.55 0.4>
    gravityOverride {z=-980}
    physicsBodyDefinitions [{
      collisionType=CustomSphere
      sphereCollisionRadius=2
      boundBone {boneName=dyn_Hair_b01}
      boxExtents=<23 23 23>
      constraintSetup {angularLimitsMax=<2 1 10>, angularLimitsMin=<-2 -7 -10>, angularTarget=<0 0 0>}
      localJointOffset=<4 0 0>
     }]
    relativeSpaceBone {boneName=root}
   }
   n35 = Anim Dynamics - Bone: dyn_Hair_b02!(SimulationSpace:AnimPhysSimSpaceType Enum, ComponentPose=@n34, Alpha=1) -> (Pose) {
    angularDampingOverride=0.99
    angularSpringConstant=100
    bAngularSpring=true
    bGravityOverrideInSimSpace=true
    bOverrideAngularDamping=true
    bUseGravityOverride=true
    simulationSpace=RootRelative
    alphaBoolBlend {blendOption=Linear}
    boundBone {boneName=dyn_Hair_b02}
    componentAppliedLinearAccClamp=<3500 3500 2500>
    componentLinearAccScale=<0.5 0.5 0.35>
    componentLinearVelScale=<0.55 0.55 0.4>
    gravityOverride {z=-980}
    physicsBodyDefinitions [{
      collisionType=CustomSphere
      sphereCollisionRadius=2
      boundBone {boneName=dyn_Hair_b02}
      boxExtents=<23 23 23>
      constraintSetup {angularLimitsMax=<2 1 10>, angularLimitsMin=<-2 -7 -10>, angularTarget=<0 0 0>}
      localJointOffset=<4 0 0>
     }]
    relativeSpaceBone {boneName=root}
   }
   n36 = Anim Dynamics - Bone: dyn_Hair_c!(SimulationSpace:AnimPhysSimSpaceType Enum, ComponentPose=@n35, Alpha=1) -> (Pose) {
    angularDampingOverride=0.99
    angularSpringConstant=100
    bAngularSpring=true
    bGravityOverrideInSimSpace=true
    bOverrideAngularDamping=true
    bUseGravityOverride=true
    simulationSpace=RootRelative
    alphaBoolBlend {blendOption=Linear}
    boundBone {boneName=dyn_Hair_c}
    componentAppliedLinearAccClamp=<3500 3500 2500>
    componentLinearAccScale=<0.5 0.5 0.35>
    componentLinearVelScale=<0.55 0.55 0.4>
    gravityOverride {z=-980}
    physicsBodyDefinitions [{
      collisionType=CustomSphere
      sphereCollisionRadius=2
      boundBone {boneName=dyn_Hair_c}
      boxExtents=<23 23 23>
      constraintSetup {angularLimitsMax=<2 1 10>, angularLimitsMin=<-2 -5 -10>, angularTarget=<0 0 0>}
      localJointOffset=<4 0 0>
     }]
    relativeSpaceBone {boneName=root}
   }
   n37 = Anim Dynamics - Bone: dyn_Hair_d01!(SimulationSpace:AnimPhysSimSpaceType Enum, ComponentPose=@n36, Alpha=1) -> (Pose) {
    angularDampingOverride=1
    angularSpringConstant=10000
    bAngularSpring=true
    bOverrideAngularDamping=true
    bUseGravityOverride=true
    gravityScale=0.35
    simulationSpace=RootRelative
    alphaBoolBlend {blendOption=Linear}
    boundBone {boneName=dyn_Hair_d01}
    componentAppliedLinearAccClamp=<3500 3500 2500>
    componentLinearAccScale=<0.5 0.5 0.35>
    componentLinearVelScale=<0.55 0.55 0.4>
    gravityOverride {z=-980}
    physicsBodyDefinitions [{
      boundBone {boneName=dyn_Hair_d01}
      boxExtents=<20 30 30>
      constraintSetup {angularLimitsMax=<0 0 5>, angularLimitsMin=<0 0 -5>, angularTarget=<1 0 0>}
      localJointOffset=<4 0 0>
     }]
   }
   n38 = Anim Dynamics - Bone: dyn_Hair_d02!(SimulationSpace:AnimPhysSimSpaceType Enum, ComponentPose=@n37, Alpha=1) -> (Pose) {
    angularDampingOverride=1
    angularSpringConstant=10000
    bAngularSpring=true
    bOverrideAngularDamping=true
    bUseGravityOverride=true
    gravityScale=0.35
    simulationSpace=RootRelative
    alphaBoolBlend {blendOption=Linear}
    boundBone {boneName=dyn_Hair_d02}
    componentAppliedLinearAccClamp=<3500 3500 2500>
    componentLinearAccScale=<0.5 0.5 0.35>
    componentLinearVelScale=<0.55 0.55 0.4>
    gravityOverride {z=-980}
    physicsBodyDefinitions [{
      boundBone {boneName=dyn_Hair_d02}
      boxExtents=<20 30 30>
      constraintSetup {angularLimitsMax=<0 0 10>, angularLimitsMin=<0 0 -10>, angularTarget=<1 0 0>}
      localJointOffset=<4 0 0>
     }]
   }
   n39 = Anim Dynamics - Bone: dyn_Hair_d03!(SimulationSpace:AnimPhysSimSpaceType Enum, ComponentPose=@n38, Alpha=1) -> (Pose) {
    angularDampingOverride=1
    angularSpringConstant=10000
    bAngularSpring=true
    bOverrideAngularDamping=true
    bUseGravityOverride=true
    gravityScale=0.35
    simulationSpace=RootRelative
    alphaBoolBlend {blendOption=Linear}
    boundBone {boneName=dyn_Hair_d03}
    componentAppliedLinearAccClamp=<3500 3500 2500>
    componentLinearAccScale=<0.5 0.5 0.35>
    componentLinearVelScale=<0.55 0.55 0.4>
    gravityOverride {z=-980}
    physicsBodyDefinitions [{
      boundBone {boneName=dyn_Hair_d03}
      boxExtents=<20 30 30>
      constraintSetup {angularLimitsMax=<0 0 10>, angularLimitsMin=<0 0 -10>, angularTarget=<1 0 0>}
      localJointOffset=<4 0 0>
     }]
   }
   n40 = Anim Dynamics - Bone: dyn_Hair_d04!(SimulationSpace:AnimPhysSimSpaceType Enum, ComponentPose=@n39, Alpha=1) -> (Pose) {
    angularDampingOverride=1
    angularSpringConstant=10000
    bAngularSpring=true
    bOverrideAngularDamping=true
    bUseGravityOverride=true
    gravityScale=0.35
    simulationSpace=RootRelative
    alphaBoolBlend {blendOption=Linear}
    boundBone {boneName=dyn_Hair_d04}
    componentAppliedLinearAccClamp=<3500 3500 2500>
    componentLinearAccScale=<0.5 0.5 0.35>
    componentLinearVelScale=<0.55 0.55 0.4>
    gravityOverride {z=-980}
    physicsBodyDefinitions [{
      boundBone {boneName=dyn_Hair_d04}
      boxExtents=<20 30 30>
      constraintSetup {angularLimitsMax=<0 0 10>, angularLimitsMin=<0 0 -10>, angularTarget=<1 0 0>}
      localJointOffset=<4 0 0>
     }]
   }
   n41 = RigidBody!(
    ComponentPose=@n40
    Alpha=0
    OverrideWorldGravity:Vector="0,0,-980"
    SimSpaceSettings:Sim Space Settings Structure=(WorldAlpha=0,VelocityScaleZ=1,DampingAlpha=1,MaxLinearVelocity=10000,MaxAngularVelocity=10000,MaxLinearAcceleration=10000,MaxAngularAcceleration=10000,ExternalLinearDragV=(X=0,Y=0,Z=0),ExternalLinearVelocity=(X=0,Y=0,Z=0),ExternalAngularVelocity=(X=0,Y=0,Z=0))
   ) -> (Pose) {
    bClampLinearTranslationLimitToRefPose=true
    bOverrideWorldGravity=true
    overridePhysicsAsset=&/F_MED_PatronPoppet_FaceAcc_Physics
    simulationSpace=BaseBoneSpace
    alphaBoolBlend {blendOption=Linear}
    baseBoneRef {boneName=spine_03}
    componentAppliedLinearAccClamp=<3500 3500 3500>
    componentLinearAccScale=<0.3 0.4 0.3>
    componentLinearVelScale=<0.8 0.8 0.8>
    overrideWorldGravity {z=-980}
    simSpaceSettings {
     maxAngularAcceleration=1000
     maxAngularVelocity=0.3
     maxLinearAcceleration=700
     maxLinearVelocity=500
     velocityScaleZ=0
     worldAlpha=0.5
     externalLinearDragV=<0.5 0.5>
    }
   }
   n42 = "Transform (Modify) Bone - Bone: dyn_Hair_BackR_03a"(ComponentPose=@n41, bAlphaBoolEnabled=True, Rotation="0,-5,10") {
    alphaInputType=Bool
    rotationMode=BMM_Additive
    rotationSpace=BCS_BoneSpace
    scaleMode=BMM_Additive
    scaleSpace=BCS_BoneSpace
    translationMode=BMM_Additive
    translationSpace=BCS_BoneSpace
    alphaBoolBlend {blendOption=Linear}
    boneToModify {boneName=dyn_Hair_BackR_03a}
    rotation {roll=10, yaw=-5}
   }
   Save cached pose 'CRCache_HairSetup'(Pose=Blend Poses by bool(BlendPose_0=Component To Local(ComponentPose=@n42), BlendPose_1=Use cached pose 'CRCache_HairPose'()) {blendType=Linear}) {cacheName=CRCache_HairSetup}
   n43 = Anim Dynamics - Bone: dyn_hat_top_01!(
    SimulationSpace:AnimPhysSimSpaceType Enum=Component
    ComponentPose=Local To Component(LocalPose=Use cached pose 'CRCache_HairSetup'())
    Alpha=1
   ) -> (Pose) {
    angularDampingOverride=1
    angularSpringConstant=20000
    bAngularSpring=true
    bOverrideAngularDamping=true
    bUseGravityOverride=true
    gravityScale=0.35
    alphaBoolBlend {blendOption=Linear}
    boundBone {boneName=dyn_hat_top_01}
    componentAppliedLinearAccClamp=<3500 3500 3500>
    componentLinearAccScale=<0.8 0.8 0.8>
    componentLinearVelScale=<0.8 0.8 0.8>
    gravityOverride {z=-980}
    physicsBodyDefinitions [{
      boundBone {boneName=dyn_hat_top_01}
      boxExtents=<30 35 35>
      constraintSetup {angularLimitsMax=<0 15 20>, angularLimitsMin=<0 -15 -20>, angularTarget=<1 0 0>}
      localJointOffset=<0.5 0 0>
     }]
    relativeSpaceBone {boneName=spine_03}
   }
   n44 = Anim Dynamics - Bone: dyn_hat_bk01!(SimulationSpace:AnimPhysSimSpaceType Enum=Component, ComponentPose=@n43, Alpha=1) -> (Pose) {
    angularDampingOverride=1
    angularSpringConstant=20000
    bAngularSpring=true
    bOverrideAngularDamping=true
    bUseGravityOverride=true
    gravityScale=0.35
    alphaBoolBlend {blendOption=Linear}
    boundBone {boneName=dyn_hat_bk01}
    componentAppliedLinearAccClamp=<4000 4000 4000>
    componentLinearAccScale=<1 1 1>
    componentLinearVelScale=<1 1 1>
    gravityOverride {z=-980}
    physicsBodyDefinitions [{
      boundBone {boneName=dyn_hat_bk01}
      boxExtents=<30 35 35>
      constraintSetup {angularLimitsMax=<0 10 0>, angularLimitsMin=<0 -15 0>, angularTarget=<1 0 0>}
      localJointOffset=<1 0 0>
     }]
    relativeSpaceBone {boneName=spine_03}
   }
   n45 = Anim Dynamics - Bone: dyn_hat_r_A01!(SimulationSpace:AnimPhysSimSpaceType Enum=Component, ComponentPose=@n44, Alpha=1) -> (Pose) {
    angularDampingOverride=1
    angularSpringConstant=20000
    bAngularSpring=true
    bOverrideAngularDamping=true
    bUseGravityOverride=true
    gravityScale=0.35
    alphaBoolBlend {blendOption=Linear}
    boundBone {boneName=dyn_hat_r_A01}
    componentAppliedLinearAccClamp=<4000 4000 4000>
    componentLinearAccScale=<1 1 1>
    componentLinearVelScale=<1 1 1>
    gravityOverride {z=-980}
    physicsBodyDefinitions [{
      boundBone {boneName=dyn_hat_r_A01}
      boxExtents=<30 35 35>
      constraintSetup {angularLimitsMax=<0 15 0>, angularLimitsMin=<0 -15 0>, angularTarget=<1 0 0>}
      localJointOffset=<1 0 0>
     }]
    relativeSpaceBone {boneName=spine_03}
   }
   n46 = Anim Dynamics - Bone: dyn_hat_r_B01!(SimulationSpace:AnimPhysSimSpaceType Enum=Component, ComponentPose=@n45, Alpha=1) -> (Pose) {
    angularDampingOverride=1
    angularSpringConstant=20000
    bAngularSpring=true
    bOverrideAngularDamping=true
    bUseGravityOverride=true
    gravityScale=0.35
    alphaBoolBlend {blendOption=Linear}
    boundBone {boneName=dyn_hat_r_B01}
    componentAppliedLinearAccClamp=<4000 4000 4000>
    componentLinearAccScale=<1 1 1>
    componentLinearVelScale=<1 1 1>
    gravityOverride {z=-980}
    physicsBodyDefinitions [{
      boundBone {boneName=dyn_hat_r_B01}
      boxExtents=<30 35 35>
      constraintSetup {angularLimitsMax=<0 15 0>, angularLimitsMin=<0 -15 0>, angularTarget=<1 0 0>}
      localJointOffset=<1 0 0>
     }]
    relativeSpaceBone {boneName=spine_03}
   }
   n47 = Anim Dynamics - Bone: dyn_hat_r_C01!(SimulationSpace:AnimPhysSimSpaceType Enum=Component, ComponentPose=@n46, Alpha=1) -> (Pose) {
    angularDampingOverride=1
    angularSpringConstant=20000
    bAngularSpring=true
    bOverrideAngularDamping=true
    bUseGravityOverride=true
    gravityScale=0.35
    alphaBoolBlend {blendOption=Linear}
    boundBone {boneName=dyn_hat_r_C01}
    componentAppliedLinearAccClamp=<4000 4000 4000>
    componentLinearAccScale=<1 1 1>
    componentLinearVelScale=<1 1 1>
    gravityOverride {z=-980}
    physicsBodyDefinitions [{
      boundBone {boneName=dyn_hat_r_C01}
      boxExtents=<30 35 35>
      constraintSetup {angularLimitsMax=<0 5 0>, angularLimitsMin=<0 -15 0>, angularTarget=<1 0 0>}
      localJointOffset=<1 0 0>
     }]
    relativeSpaceBone {boneName=spine_03}
   }
   n48 = Anim Dynamics - Bone: dyn_hat_fn01!(SimulationSpace:AnimPhysSimSpaceType Enum=Component, ComponentPose=@n47, Alpha=1) -> (Pose) {
    angularDampingOverride=1
    angularSpringConstant=20000
    bAngularSpring=true
    bOverrideAngularDamping=true
    bUseGravityOverride=true
    gravityScale=0.35
    alphaBoolBlend {blendOption=Linear}
    boundBone {boneName=dyn_hat_fn01}
    componentAppliedLinearAccClamp=<4000 4000 4000>
    componentLinearAccScale=<1 1 1>
    componentLinearVelScale=<1 1 1>
    gravityOverride {z=-980}
    physicsBodyDefinitions [{
      boundBone {boneName=dyn_hat_fn01}
      boxExtents=<30 35 35>
      constraintSetup {angularLimitsMax=<0 15 0>, angularLimitsMin=<0 -15 0>, angularTarget=<1 0 0>}
      localJointOffset=<1 0 0>
     }]
    relativeSpaceBone {boneName=spine_03}
   }
   n49 = Anim Dynamics - Bone: dyn_hat_l_A01!(SimulationSpace:AnimPhysSimSpaceType Enum=Component, ComponentPose=@n48, Alpha=1) -> (Pose) {
    angularDampingOverride=1
    angularSpringConstant=20000
    bAngularSpring=true
    bOverrideAngularDamping=true
    bUseGravityOverride=true
    gravityScale=0.35
    alphaBoolBlend {blendOption=Linear}
    boundBone {boneName=dyn_hat_l_A01}
    componentAppliedLinearAccClamp=<4000 4000 4000>
    componentLinearAccScale=<1 1 1>
    componentLinearVelScale=<1 1 1>
    gravityOverride {z=-980}
    physicsBodyDefinitions [{
      boundBone {boneName=dyn_hat_l_A01}
      boxExtents=<30 35 35>
      constraintSetup {angularLimitsMax=<0 15 0>, angularLimitsMin=<0 -15 0>, angularTarget=<1 0 0>}
      localJointOffset=<1 0 0>
     }]
    relativeSpaceBone {boneName=spine_03}
   }
   n50 = Anim Dynamics - Bone: dyn_hat_l_B01!(SimulationSpace:AnimPhysSimSpaceType Enum=Component, ComponentPose=@n49, Alpha=1) -> (Pose) {
    angularDampingOverride=1
    angularSpringConstant=20000
    bAngularSpring=true
    bOverrideAngularDamping=true
    bUseGravityOverride=true
    gravityScale=0.35
    alphaBoolBlend {blendOption=Linear}
    boundBone {boneName=dyn_hat_l_B01}
    componentAppliedLinearAccClamp=<4000 4000 4000>
    componentLinearAccScale=<1 1 1>
    componentLinearVelScale=<1 1 1>
    gravityOverride {z=-980}
    physicsBodyDefinitions [{
      boundBone {boneName=dyn_hat_l_B01}
      boxExtents=<30 35 35>
      constraintSetup {angularLimitsMax=<0 15 0>, angularLimitsMin=<0 -15 0>, angularTarget=<1 0 0>}
      localJointOffset=<1 0 0>
     }]
    relativeSpaceBone {boneName=spine_03}
   }
   n51 = Anim Dynamics - Bone: dyn_hat_l_C01!(SimulationSpace:AnimPhysSimSpaceType Enum=Component, ComponentPose=@n50, Alpha=1) -> (Pose) {
    angularDampingOverride=1
    angularSpringConstant=20000
    bAngularSpring=true
    bOverrideAngularDamping=true
    bUseGravityOverride=true
    gravityScale=0.35
    alphaBoolBlend {blendOption=Linear}
    boundBone {boneName=dyn_hat_l_C01}
    componentAppliedLinearAccClamp=<4000 4000 4000>
    componentLinearAccScale=<1 1 1>
    componentLinearVelScale=<1 1 1>
    gravityOverride {z=-980}
    physicsBodyDefinitions [{
      boundBone {boneName=dyn_hat_l_C01}
      boxExtents=<30 35 35>
      constraintSetup {angularLimitsMax=<0 5 0>, angularLimitsMin=<0 -15 0>, angularTarget=<1 0 0>}
      localJointOffset=<1 0 0>
     }]
    relativeSpaceBone {boneName=spine_03}
   }
   n52 = Anim Dynamics - Bone: dyn_branch_A01!(SimulationSpace:AnimPhysSimSpaceType Enum=BoneRelative, ComponentPose=@n51, Alpha=1) -> (Pose) {
    angularDampingOverride=1
    angularSpringConstant=20000
    bAngularSpring=true
    bOverrideAngularDamping=true
    bUseGravityOverride=true
    gravityScale=0.35
    simulationSpace=BoneRelative
    alphaBoolBlend {blendOption=Linear}
    boundBone {boneName=dyn_branch_A01}
    componentAppliedLinearAccClamp=<4000 40000 40000>
    componentLinearAccScale=<1 1 1>
    componentLinearVelScale=<1 1 1>
    gravityOverride {z=-980}
    physicsBodyDefinitions [{
      boundBone {boneName=dyn_branch_A01}
      boxExtents=<20 30 30>
      constraintSetup {angularLimitsMax=<0 5 5>, angularLimitsMin=<0 -5 -5>, angularTarget=<1 0 0>}
      localJointOffset=<4 0 0>
     }]
    relativeSpaceBone {boneName=spine_03}
   }
   n53 = Anim Dynamics - Bone: dyn_branch_A02!(SimulationSpace:AnimPhysSimSpaceType Enum=BoneRelative, ComponentPose=@n52, Alpha=1) -> (Pose) {
    angularDampingOverride=1
    angularSpringConstant=20000
    bAngularSpring=true
    bOverrideAngularDamping=true
    bUseGravityOverride=true
    gravityScale=0.35
    simulationSpace=BoneRelative
    alphaBoolBlend {blendOption=Linear}
    boundBone {boneName=dyn_branch_A02}
    componentAppliedLinearAccClamp=<4000 40000 40000>
    componentLinearAccScale=<1 1 1>
    componentLinearVelScale=<1 1 1>
    gravityOverride {z=-980}
    physicsBodyDefinitions [{
      boundBone {boneName=dyn_branch_A02}
      boxExtents=<20 30 30>
      constraintSetup {angularLimitsMax=<0 5 5>, angularLimitsMin=<0 -5 -5>, angularTarget=<1 0 0>}
      localJointOffset=<4 0 0>
     }]
    relativeSpaceBone {boneName=spine_03}
   }
   n54 = Anim Dynamics - Bone: dyn_branch_B01!(SimulationSpace:AnimPhysSimSpaceType Enum=BoneRelative, ComponentPose=@n53, Alpha=1) -> (Pose) {
    angularDampingOverride=1
    angularSpringConstant=20000
    bAngularSpring=true
    bOverrideAngularDamping=true
    bUseGravityOverride=true
    gravityScale=0.35
    simulationSpace=BoneRelative
    alphaBoolBlend {blendOption=Linear}
    boundBone {boneName=dyn_branch_B01}
    componentAppliedLinearAccClamp=<4000 40000 40000>
    componentLinearAccScale=<1 1 1>
    componentLinearVelScale=<1 1 1>
    gravityOverride {z=-980}
    physicsBodyDefinitions [{
      boundBone {boneName=dyn_branch_B01}
      boxExtents=<20 30 30>
      constraintSetup {angularLimitsMax=<0 0 2>, angularLimitsMin=<0 -2 -2>, angularTarget=<1 0 0>}
      localJointOffset=<4 0 0>
     }]
    relativeSpaceBone {boneName=spine_03}
   }
   n55 = Anim Dynamics - Bone: dyn_branch_B02!(SimulationSpace:AnimPhysSimSpaceType Enum=BoneRelative, ComponentPose=@n54, Alpha=1) -> (Pose) {
    angularDampingOverride=1
    angularSpringConstant=20000
    bAngularSpring=true
    bOverrideAngularDamping=true
    bUseGravityOverride=true
    gravityScale=0.35
    simulationSpace=BoneRelative
    alphaBoolBlend {blendOption=Linear}
    boundBone {boneName=dyn_branch_B02}
    componentAppliedLinearAccClamp=<4000 40000 40000>
    componentLinearAccScale=<1 1 1>
    componentLinearVelScale=<1 1 1>
    gravityOverride {z=-980}
    physicsBodyDefinitions [{
      boundBone {boneName=dyn_branch_B02}
      boxExtents=<20 30 30>
      constraintSetup {angularLimitsMax=<0 0 2>, angularLimitsMin=<0 -2 -2>, angularTarget=<1 0 0>}
      localJointOffset=<4 0 0>
     }]
    relativeSpaceBone {boneName=spine_03}
   }
   n56 = Anim Dynamics - Bone: dyn_pumpkinleaf_A01!(SimulationSpace:AnimPhysSimSpaceType Enum=BoneRelative, ComponentPose=@n55, Alpha=1) -> (Pose) {
    angularDampingOverride=1
    angularSpringConstant=20000
    bAngularSpring=true
    bOverrideAngularDamping=true
    bUseGravityOverride=true
    gravityScale=0.35
    simulationSpace=BoneRelative
    alphaBoolBlend {blendOption=Linear}
    boundBone {boneName=dyn_pumpkinleaf_A01}
    componentAppliedLinearAccClamp=<5000 5000 5000>
    componentLinearAccScale=<1 1 1>
    componentLinearVelScale=<1 1 1>
    gravityOverride {z=-980}
    physicsBodyDefinitions [{
      boundBone {boneName=dyn_pumpkinleaf_A01}
      boxExtents=<20 30 30>
      constraintSetup {angularLimitsMax=<0 2 2>, angularLimitsMin=<0 -2 -2>, angularTarget=<1 0 0>}
      localJointOffset=<4 0 0>
     }]
    relativeSpaceBone {boneName=spine_03}
   }
   n57 = Anim Dynamics - Bone: dyn_pumpkinleaf_A02!(SimulationSpace:AnimPhysSimSpaceType Enum=BoneRelative, ComponentPose=@n56, Alpha=1) -> (Pose) {
    angularDampingOverride=1
    angularSpringConstant=20000
    bAngularSpring=true
    bOverrideAngularDamping=true
    bUseGravityOverride=true
    gravityScale=0.35
    simulationSpace=BoneRelative
    alphaBoolBlend {blendOption=Linear}
    boundBone {boneName=dyn_pumpkinleaf_A02}
    componentAppliedLinearAccClamp=<5000 5000 5000>
    componentLinearAccScale=<1 1 1>
    componentLinearVelScale=<1 1 1>
    gravityOverride {z=-980}
    physicsBodyDefinitions [{
      boundBone {boneName=dyn_pumpkinleaf_A02}
      boxExtents=<20 30 30>
      constraintSetup {angularLimitsMax=<0 2 2>, angularLimitsMin=<0 -2 -2>, angularTarget=<1 0 0>}
      localJointOffset=<4 0 0>
     }]
    relativeSpaceBone {boneName=spine_03}
   }
   n58 = Anim Dynamics - Bone: dyn_pumpkinleaf_B01!(SimulationSpace:AnimPhysSimSpaceType Enum=BoneRelative, ComponentPose=@n57, Alpha=1) -> (Pose) {
    angularDampingOverride=1
    angularSpringConstant=20000
    bAngularSpring=true
    bOverrideAngularDamping=true
    bUseGravityOverride=true
    gravityScale=0.35
    simulationSpace=BoneRelative
    alphaBoolBlend {blendOption=Linear}
    boundBone {boneName=dyn_pumpkinleaf_B01}
    componentAppliedLinearAccClamp=<5000 5000 5000>
    componentLinearAccScale=<1 1 1>
    componentLinearVelScale=<1 1 1>
    gravityOverride {z=-980}
    physicsBodyDefinitions [{
      boundBone {boneName=dyn_pumpkinleaf_B01}
      boxExtents=<20 30 30>
      constraintSetup {angularLimitsMax=<0 2 2>, angularLimitsMin=<0 -2 -2>, angularTarget=<1 0 0>}
      localJointOffset=<4 0 0>
     }]
    relativeSpaceBone {boneName=spine_03}
   }
   n59 = Anim Dynamics - Bone: dyn_pumpkinleaf_B02!(SimulationSpace:AnimPhysSimSpaceType Enum=BoneRelative, ComponentPose=@n58, Alpha=1) -> (Pose) {
    angularDampingOverride=1
    angularSpringConstant=20000
    bAngularSpring=true
    bOverrideAngularDamping=true
    bUseGravityOverride=true
    gravityScale=0.35
    simulationSpace=BoneRelative
    alphaBoolBlend {blendOption=Linear}
    boundBone {boneName=dyn_pumpkinleaf_B02}
    componentAppliedLinearAccClamp=<5000 5000 5000>
    componentLinearAccScale=<1 1 1>
    componentLinearVelScale=<1 1 1>
    gravityOverride {z=-980}
    physicsBodyDefinitions [{
      boundBone {boneName=dyn_pumpkinleaf_B02}
      boxExtents=<20 30 30>
      constraintSetup {angularLimitsMax=<0 2 2>, angularLimitsMin=<0 -2 -2>, angularTarget=<1 0 0>}
      localJointOffset=<4 0 0>
     }]
    relativeSpaceBone {boneName=spine_03}
   }
   Save cached pose 'CRCache_HatDyn'(Pose=Component To Local(ComponentPose=@n59)) {cacheName=CRCache_HatDyn}
   n60 = "Transform (Modify) Bone - Bone: base_hat_m"!(ComponentPose=Local To Component(LocalPose=Use cached pose 'CRCache_HatDyn'()), bAlphaBoolEnabled:bool=True, Rotation="3,0,0") -> (Pose) {
    alphaInputType=Bool
    rotationMode=BMM_Additive
    rotationSpace=BCS_BoneSpace
    scaleMode=BMM_Additive
    scaleSpace=BCS_BoneSpace
    translationMode=BMM_Additive
    translationSpace=BCS_BoneSpace
    alphaBoolBlend {blendOption=Linear}
    boneToModify {boneName=base_hat_m}
    rotation {pitch=3}
   }
   n61 = "Transform (Modify) Bone - Bone: dyn_Hair_BackR_03"!(ComponentPose=@n60, bAlphaBoolEnabled:bool=True, Scale="0.7,1,0.7") -> (Pose) {
    alphaInputType=Bool
    rotationMode=BMM_Additive
    rotationSpace=BCS_BoneSpace
    scaleMode=BMM_Additive
    scaleSpace=BCS_BoneSpace
    translationMode=BMM_Additive
    translationSpace=BCS_BoneSpace
    alphaBoolBlend {blendInTime=0.1, blendOption=Linear, blendOutTime=0.1}
    boneToModify {boneName=dyn_Hair_BackR_03}
    scale {x=0.7, z=0.7}
   }
   n62 = "Transform (Modify) Bone - Bone: dyn_Hair_Back_04"!(ComponentPose=@n61, bAlphaBoolEnabled:bool=True, Rotation="-3,0,0") -> (Pose) {
    alphaInputType=Bool
    rotationMode=BMM_Additive
    rotationSpace=BCS_BoneSpace
    scaleMode=BMM_Additive
    scaleSpace=BCS_BoneSpace
    translationMode=BMM_Additive
    translationSpace=BCS_BoneSpace
    alphaBoolBlend {blendInTime=0.1, blendOption=Linear, blendOutTime=0.1}
    boneToModify {boneName=dyn_Hair_Back_04}
    rotation {pitch=-3}
   }
   n63 = "Transform (Modify) Bone - Bone: dyn_Hair_Back_06"!(ComponentPose=@n62, bAlphaBoolEnabled:bool=True, Scale="0.8,1,1") -> (Pose) {
    alphaInputType=Bool
    rotationMode=BMM_Additive
    rotationSpace=BCS_BoneSpace
    scaleMode=BMM_Additive
    scaleSpace=BCS_BoneSpace
    translationMode=BMM_Additive
    translationSpace=BCS_BoneSpace
    alphaBoolBlend {blendInTime=0.1, blendOption=Linear, blendOutTime=0.1}
    boneToModify {boneName=dyn_Hair_Back_06}
    scale {x=0.8}
   }
   n64 = "Transform (Modify) Bone - Bone: dyn_Hair_Back_07"!(ComponentPose=@n63, bAlphaBoolEnabled:bool=True, Rotation="15,0,0") -> (Pose) {
    alphaInputType=Bool
    rotationMode=BMM_Additive
    rotationSpace=BCS_BoneSpace
    scaleMode=BMM_Additive
    scaleSpace=BCS_BoneSpace
    translationMode=BMM_Additive
    translationSpace=BCS_BoneSpace
    alphaBoolBlend {blendOption=Linear}
    boneToModify {boneName=dyn_Hair_Back_07}
    rotation {pitch=15}
   }
   Save cached pose 'CRCache_BP'(Pose=Component To Local(ComponentPose=@n64)) {cacheName=CRCache_BP}
   n65 = "Transform (Modify) Bone - Bone: dyn_Hair_Back_04"!(
    ComponentPose=Local To Component(LocalPose=Use cached pose 'CRCache_HatDyn'())
    bAlphaBoolEnabled=True
    Rotation="-5,0,10"
    Scale:Vector="0.9,1,0.7"
   ) -> (Pose) {
    alphaInputType=Bool
    rotationMode=BMM_Additive
    rotationSpace=BCS_BoneSpace
    scaleMode=BMM_Additive
    scaleSpace=BCS_BoneSpace
    translationMode=BMM_Additive
    translationSpace=BCS_BoneSpace
    alphaBoolBlend {blendOption=Linear}
    boneToModify {boneName=dyn_Hair_Back_04}
    rotation {pitch=-5, roll=10}
    scale {x=0.9, z=0.7}
   }
   n66 = "Transform (Modify) Bone - Bone: dyn_Hair_BackL_04"!(ComponentPose=@n65, bAlphaBoolEnabled:bool=True, Rotation="-10,0,0", Scale="0.9,1,0.5") -> (Pose) {
    alphaInputType=Bool
    rotationMode=BMM_Additive
    rotationSpace=BCS_BoneSpace
    scaleMode=BMM_Additive
    scaleSpace=BCS_BoneSpace
    translationMode=BMM_Additive
    translationSpace=BCS_BoneSpace
    alphaBoolBlend {blendOption=Linear}
    boneToModify {boneName=dyn_Hair_BackL_04}
    rotation {pitch=-10}
    scale {x=0.9, z=0.5}
   }
   n67 = "Transform (Modify) Bone - Bone: dyn_Hair_BackR_04"!(ComponentPose=@n66, bAlphaBoolEnabled:bool=True, Rotation="-10,0,0", Scale="0.9,1,0.7") -> (Pose) {
    alphaInputType=Bool
    rotationMode=BMM_Additive
    rotationSpace=BCS_BoneSpace
    scaleMode=BMM_Additive
    scaleSpace=BCS_BoneSpace
    translationMode=BMM_Additive
    translationSpace=BCS_BoneSpace
    alphaBoolBlend {blendOption=Linear}
    boneToModify {boneName=dyn_Hair_BackR_04}
    rotation {pitch=-10}
    scale {x=0.9, z=0.7}
   }
   n68 = "Transform (Modify) Bone - Bone: dyn_Hair_Back_04"!(ComponentPose=@n67, bAlphaBoolEnabled=True, Translation:Vector="0,0,-2", Rotation="-5,10,0", Scale="0.7,1,0.9") -> (Pose) {
    alphaInputType=Bool
    rotationMode=BMM_Additive
    rotationSpace=BCS_BoneSpace
    scaleMode=BMM_Additive
    scaleSpace=BCS_BoneSpace
    translationMode=BMM_Additive
    translationSpace=BCS_BoneSpace
    alphaBoolBlend {blendInTime=0.05, blendOption=Linear, blendOutTime=0.2}
    boneToModify {boneName=dyn_Hair_Back_04}
    rotation {pitch=-5, yaw=10}
    scale {x=0.7, z=0.9}
    translation {z=-2}
   }
   n69 = "Transform (Modify) Bone - Bone: dyn_Hair_BackL_04"!(ComponentPose=@n68, bAlphaBoolEnabled=True, Translation:Vector="0,0,-2", Rotation="-5,0,0", Scale="0.7,1,0.9") -> (Pose) {
    alphaInputType=Bool
    rotationMode=BMM_Additive
    rotationSpace=BCS_BoneSpace
    scaleMode=BMM_Additive
    scaleSpace=BCS_BoneSpace
    translationMode=BMM_Additive
    translationSpace=BCS_BoneSpace
    alphaBoolBlend {blendInTime=0.05, blendOption=Linear, blendOutTime=0.2}
    boneToModify {boneName=dyn_Hair_BackL_04}
    rotation {pitch=-5}
    scale {x=0.7, z=0.9}
    translation {z=-2}
   }
   n70 = "Transform (Modify) Bone - Bone: dyn_Hair_BackR_04"(ComponentPose=@n69, bAlphaBoolEnabled=True, Rotation="0,10,0", Scale="0.7,1,0.9") {
    alphaInputType=Bool
    rotationMode=BMM_Additive
    rotationSpace=BCS_BoneSpace
    scaleMode=BMM_Additive
    scaleSpace=BCS_BoneSpace
    translationMode=BMM_Additive
    translationSpace=BCS_BoneSpace
    alphaBoolBlend {blendInTime=0.05, blendOption=Linear, blendOutTime=0.2}
    boneToModify {boneName=dyn_Hair_BackR_04}
    rotation {yaw=10}
    scale {x=0.7, z=0.9}
   }
   Save cached pose 'CRCache_Cape'(Pose=Component To Local(ComponentPose=@n70)) {cacheName=CRCache_Cape}
  ]}
 EventGraph {nodes [
   n1 = Event BlueprintUpdateAnimation()
   > n2 = Sequence()
   > n3 = Gravity Override State Function DEPRECATED(CharacterPartAnimInstance=Self-Reference(), GameTimeInSeconds=GetGameTimeInSeconds(), GravityOverrideParams=Get Gravity Override Params())
   > Set Gravity Override(Gravity Override=@n3)
   n4 = Get isWearingPackBool() -> (isWearingPackBool:bool)
   n5 = Get bIsDBNO()
   > Set DBNO_Cape_Float(DBNO_Cape_Float:double=SelectFloat(A=1, B=0.2, bPickA=OR Boolean(A=@n4, B=@n5))) -> (Output_Get:double)
   > Set DivePack(DivePack:bool=OR Boolean!(A=Get bIsSkydiveDiveMode(), B=Get bIsSkydiveFloating(), C:bool=Get isWearingPackBool(), D:bool=Get bIsDBNO()) -> (ReturnValue)) -> (Output_Get:bool)
   > Set CrouchPack(CrouchPack:bool=OR Boolean!(A=Get bIsCrouching(), B=Get isWearingPackBool()) -> (ReturnValue)) -> (Output_Get:bool)
   > Set FE_Skydive(FE_Skydive:bool=OR Boolean(A=Get bIsSkyDiving(), B=Get bIsInFrontEnd())) -> (Output_Get:bool)
   n6 = Get bIsJumping()
   n7 = Get bIsFalling()
   > Set isTacticalJump(isTacticalJump:bool=AND Boolean(A=Get bIsTacticalSprinting(), B=OR Boolean(A=@n6, B=@n7))) -> (Output_Get:bool)
   > Set isTacticalSprintJump(isTacticalSprintJump:bool=OR Boolean(A=Get isTacticalJump() -> (isTacticalJump:bool), B=Get bIsTacticalSprinting())) -> (Output_Get:bool)
   n8 = OR Boolean(A=Get bIsSkydiveFloating(), B=Get bIsSkydiveDiveMode())
   Set isSkydiveFloating(
    execute=@n2.then_1
    isSkydiveFloating:double=FInterpTo(Current=Get isSkydiveFloating(), Target="To Float (Boolean)"(InBool=@n8), DeltaTime=@n1.DeltaTimeX, InterpSpeed=8)
   ) -> (Output_Get:double)
   n9 = Get bIsCrouching()
   > Set isCrouching(isCrouching:double=FInterpTo(Current=Get isCrouching(), Target="To Float (Boolean)"(InBool=@n9), DeltaTime=@n1.DeltaTimeX, InterpSpeed=15)) -> (Output_Get:double)
   n10 = Get bIsTargeting()
   > Set isTargeting(isTargeting:double=FInterpTo(Current=Get isTargeting(), Target="To Float (Boolean)"(InBool=@n10), DeltaTime=@n1.DeltaTimeX, InterpSpeed=15)) -> (Output_Get:double)
   > n11 = Sequence()
   > n12 = Branch(Condition=Get bIsInFrontEnd())
   > Set BkHairDynAlpha(BkHairDynAlpha:double=0.15) -> (Output_Get:double)
   n13 = Branch(execute=@n12.else, Condition=OR Boolean!(A=Get bIsSkydiveDiveMode(), B=Get bIsSkydiveFloating(), C:bool=Get bIsCrouching()) -> (ReturnValue))
   > Set BkHairDynAlpha(BkHairDynAlpha=0.3)
   Set BkHairDynAlpha(execute=@n13.else, BkHairDynAlpha=0.8)
   n14 = Branch(execute=@n11.then_1, Condition=Get bIsInFrontEnd())
   > Set BangsDynAlpha(BangsDynAlpha:double=0.5) -> (Output_Get:double)
   > Set HatDynAlpha(HatDynAlpha:double=0.25) -> (Output_Get:double)
   > Set FnHairDynAlpha(FnHairDynAlpha:double=0.35) -> (Output_Get:double)
   Set BangsDynAlpha(execute=@n14.else, BangsDynAlpha=1)
   > Set HatDynAlpha(HatDynAlpha=1)
   > n15 = Branch(Condition=Get bIsParachuteOpened())
   > Set FnHairDynAlpha(FnHairDynAlpha=0.5)
   Set FnHairDynAlpha(execute=@n15.else, FnHairDynAlpha=1)
   Event OnParachuteClosed()
   > Set isHangGliding(isHangGliding:double=0) -> (Output_Get:double)
   n16 = Event OnCharacterCustomizationLoaded()
   > Set Skel_Mesh(Skel_Mesh=@n16.CustomBaseMesh)
   > n17 = Cast To PlayerPawn_Athena(Object=TryGetPawnOwner())
   > Set Body_Mesh(Body_Mesh=GetSkeletalMeshForPartType(self=@n17))
   n18 = CheckForBackbling()
   > Set isWearingPackBool(isWearingPackBool:bool=OR Boolean(A=@n18, B=@n18.bIsBackblingACape, C=@n18.bIsBackblingLong)) -> (Output_Get:bool)
   n19 = OR Boolean!(A=@n18.bIsBackblingACape, B=@n18.bIsBackblingLong) -> (ReturnValue)
   > Set is_f_WearingCape(is_f_WearingCape:double="To Float (Boolean)"(InBool=@n19)) -> (Output_Get:double)
   > Set isWearingBP(isWearingBP:bool=AND Boolean(A=@n18, B=NOT Boolean(A=@n19))) -> (Output_Get:bool)
   > n20 = Branch(Condition=Get bIsInFrontEnd())
   > Set Sim Space A(Sim Space A:AnimPhysSimSpaceType Enum=World) -> (Output_Get:AnimPhysSimSpaceType Enum)
   > Set Sim Space B(Sim Space B:AnimPhysSimSpaceType Enum=World) -> (Output_Get:AnimPhysSimSpaceType Enum)
   > Set Sim Space C(Sim Space C:AnimPhysSimSpaceType Enum=World) -> (Output_Get:AnimPhysSimSpaceType Enum)
   > Set Sim Space RB(Sim Space RB:Sim Space Settings Structure=Get Sim Space FE() -> (Sim Space FE:Sim Space Settings Structure)) -> (Output_Get:Sim Space Settings Structure)
   Set Sim Space A(execute=@n20.else)
   > Set Sim Space B(Sim Space B=BoneRelative)
   > Set Sim Space C(Sim Space C=Component)
   > Set Sim Space RB(Sim Space RB=Get Sim Space BR() -> (Sim Space BR:Sim Space Settings Structure))
   Event OnParachuteOpened()
   > Switch on EGliderType(Selection=Get PlayerGliderType() -> (PlayerGliderType=HangGlider))
   > Set isHangGliding(isHangGliding=1)
   Event OnWeaponChanged()
   > Set isGunOut(isGunOut:double=1) -> (Output_Get:double)
   > n21 = Switch on EFortWeaponCoreAnimation!(execute, Selection:EFortWeaponCoreAnimation Enum=Get WeaponCoreAnim!() -> (WeaponCoreAnim)) -> (
    Melee:Exec
    Pistol:Exec
    Shotgun:Exec
    PaperBlueprint:Exec
    Rifle:Exec
    MeleeOneHand:Exec
    MachinePistol:Exec
    RocketLauncher:Exec
    GrenadeLauncher:Exec
    GoingCommando:Exec
    AssaultRifle:Exec
    TacticalShotgun:Exec
    SniperRifle:Exec
    TrapPlacement:Exec
    ShoulderLauncher:Exec
    AbilityDecoTool:Exec
    Crossbow:Exec
    C4:Exec
    RemoteControl:Exec
    DualWield:Exec
    AR_BullPup:Exec
    AR_ForwardGrip:Exec
    MedPackPaddles:Exec
    SMG_P90:Exec
    AR_DrumGun:Exec
    Consumable_Small:Exec
    Consumable_Large:Exec
    Balloon:Exec
    MountedTurret:Exec
    CreativeTool:Exec
    ExplosiveBow:Exec
    AshtonIndigo:Exec
    AshtonChicago:Exec
    MeleeDualWield:Exec
    Unarmed:Exec
   )
   Set isGunOut(execute=[@n21.Melee, @n21.MeleeDualWield, @n21.MeleeOneHand, @n21.Unarmed], isGunOut=0)
  ]}
]
variables [
 BangsDynAlpha {type=number}
 BkHairDynAlpha {type=number}
 Body_Mesh {type=SkeletalMeshComponent}
 CrouchPack {type=boolean}
 DBNO_Cape_Float {type=number}
 DivePack {type=boolean}
 FE_Skydive {type=boolean}
 FnHairDynAlpha {type=number}
 Gravity Override {type=Vector}
 Gravity Override Params {type=GravityOverrideParamsStruct}
 HatDynAlpha {type=number}
 is_f_WearingCape {type=number}
 is_WearingCape {type=boolean}
 isCrouching {type=number}
 isGunOut {type=number}
 isHangGliding {type=number}
 isSkydiveFloating {type=number}
 isTacticalJump {type=boolean}
 isTacticalSprintJump {type=boolean}
 isTargeting {type=number}
 isWearingBP {type=boolean}
 isWearingPackBool {type=boolean}
 Sim Space A {type=AnimPhysSimSpaceType}
 Sim Space B {type=AnimPhysSimSpaceType}
 Sim Space BR {
  type=SimSpaceSettings
  default {
   dampingAlpha=1
   maxAngularAcceleration=1000
   maxAngularVelocity=0.3
   maxLinearAcceleration=700
   maxLinearVelocity=500
   velocityScaleZ=0
   worldAlpha=0.1
   externalLinearDragV=<0.5 0.5 0>
  }
 }
 Sim Space C {type=AnimPhysSimSpaceType}
 Sim Space FE {
  type=SimSpaceSettings
  default {
   dampingAlpha=1
   maxAngularAcceleration=1000
   maxAngularVelocity=3
   maxLinearAcceleration=700
   maxLinearVelocity=500
   velocityScaleZ=1
   worldAlpha=0.5
   externalLinearDragV=<0.2 0.2 0.1>
  }
 }
 Sim Space RB {
  type=SimSpaceSettings
  default {
   dampingAlpha=1
   maxAngularAcceleration=1000
   maxAngularVelocity=0.3
   maxLinearAcceleration=700
   maxLinearVelocity=150
   velocityScaleZ=0
   worldAlpha=0.5
   externalLinearDragV=<0.5 0.5 0>
  }
 }
 Skel_Mesh {type=SkeletalMeshComponent}
]
